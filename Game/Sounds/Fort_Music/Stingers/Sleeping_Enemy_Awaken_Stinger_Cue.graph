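# Sound cue for the music stinger named for a sleeping enemy waking up, playing a negative onboarding orchestral stinger.
# A single wave player feeds the output at boosted volume in the music sound class.
# It uses husk monster attenuation, so the music is treated as a spatial sound from the enemy.
# Its concurrency override allows one instance at a time and blocks new plays while it sounds.
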SoundCue /Game/Sounds/Fort_Music/Stingers/Sleeping_Enemy_Awaken_Stinger_Cue
root=WavePlayer_0
nodes [{
  id=WavePlayer_0
  properties {soundWaveAssetPtr=/Game/Sounds/Fort_Music/Orchestral_Score/Orch_Stingers/Stingers_Onboarding_Stinger_Negative_148BPM-001}
 }]
properties {
 attenuationSettings=/Game/Sounds/Attenuation/Fort_Monster_Husk_Att
 bOverrideConcurrency=true
 maxDistance=2800
 soundClassObject=/Game/Sounds/FortSoundClassesAndModes/Fort_Music
 volumeMultiplier=2
 concurrencyOverrides {maxCount=1, resolutionRule=PreventNew}
}
