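# Victory Royale music stinger Sound Cue, with no root node or sound nodes, so it holds no audio to play.
# Only cue settings remain: it routes to the base music sound class with a raised volume multiplier and a very large max distance.
# A concurrency override limits it to one instance and prevents new plays while one is active.
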SoundCue /Game/Sounds/Fort_Music/Stingers/Victory_Royale_Sound_Cue
root=null
nodes []
properties {
 bOverrideConcurrency=true
 maxDistance=2097152
 soundClassObject=/Game/Sounds/FortSoundClassesAndModes/Fort_11_Base_Music.Fort_11_Base_music
 volumeMultiplier=2
 concurrencyOverrides {maxCount=1, resolutionRule=PreventNew}
}
